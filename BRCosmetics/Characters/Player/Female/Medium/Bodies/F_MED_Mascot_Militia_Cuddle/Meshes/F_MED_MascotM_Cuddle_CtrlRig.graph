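# Control Rig for the female medium Mascot Militia Cuddle body that procedurally adjusts the jacket collar's dynamic bones.
# Right upper arm rotation weights an additive offset on the right side collar.
# A FrontEnd bool then selects either fixed back-collar rotations or two head-versus-pelvis spherical pose readers weighting collar rotations.
# A shared function adds weighted local offsets; other modify chains and debug draws sit unconnected to execution.
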ControlRigBlueprint /BRCosmetics/Characters/Player/Female/Medium/Bodies/F_MED_Mascot_Militia_Cuddle/Meshes/F_MED_MascotM_Cuddle_CtrlRig
graphs [
 ContainedGraph {nodes [
   n1 = RigVMFunctionEntryNode()
   RigUnit_GetTransform = &RigUnit_GetTransform(Space=LocalSpace, bInitial=false, Item=@n1.Joint)
   n2 = &RigVMFunction_MathQuaternionToEuler(RotationOrder=ZYX, Value=@RigUnit_GetTransform.Transform.Rotation)
   > RigUnit_SetTransform = &RigUnit_SetTransform(
    Space=LocalSpace
    Value=(Rotation=(X=0.0132,Y=0.4099,Z=-0.0222,W=0.9118),Translation=(X=23.6045,Y=-6.37,Z=140.1115),Scale3D=(X=1,Y=1,Z=1))
    bInitial=false
    Item=@n1.Joint
    Weight=@n1.Weight
    Value.Translation.Z=&RigVMFunction_MathFloatAdd(B=@n1.Translate Z, A=@RigUnit_GetTransform.Transform.Translation.Z) {template="Add::Execute(in A,in B,out Result)"}.Result
    Value.Scale3D.Z=&RigVMFunction_MathFloatAdd(B=@n1.Scale Z, A=@RigUnit_GetTransform.Transform.Scale3D.Z) {template="Add::Execute(in A,in B,out Result)"}.Result
    Value.Translation.Y=&RigVMFunction_MathFloatAdd(B=@n1.Translate Y, A=@RigUnit_GetTransform.Transform.Translation.Y) {template="Add::Execute(in A,in B,out Result)"}.Result
    Value.Scale3D.Y=&RigVMFunction_MathFloatAdd(B=@n1.Scale Y, A=@RigUnit_GetTransform.Transform.Scale3D.Y) {template="Add::Execute(in A,in B,out Result)"}.Result
    Value.Translation.X=&RigVMFunction_MathFloatAdd(B=@n1.Translate X, A=@RigUnit_GetTransform.Transform.Translation.X) {template="Add::Execute(in A,in B,out Result)"}.Result
    Value.Scale3D.X=&RigVMFunction_MathFloatAdd(B=@n1.Scale X, A=@RigUnit_GetTransform.Transform.Scale3D.X) {template="Add::Execute(in A,in B,out Result)"}.Result
    Value.Rotation=&RigVMFunction_MathQuaternionFromEuler(
     Euler=(X=0,Y=0,Z=0)
     Euler.X=&RigVMFunction_MathFloatAdd(B=@n1.Rotate X, A=@n2.Result.X) {template="Add::Execute(in A,in B,out Result)"}.Result
     Euler.Y=&RigVMFunction_MathFloatAdd(B=@n1.Rotate Y, A=@n2.Result.Y) {template="Add::Execute(in A,in B,out Result)"}.Result
     Euler.Z=&RigVMFunction_MathFloatAdd(B=@n1.Rotate Z, A=@n2.Result.Z) {template="Add::Execute(in A,in B,out Result)"}.Result
    ).Result
   )
   > RigVMFunctionReturnNode()
  ]}
 RigVMFunctionLibrary {nodes [Set Transforms for Bone = RigVMCollapseNode(
    ExecuteContext=()
    Joint=()
    Rotate X=0
    Rotate Y=0
    Rotate Z=0
    Scale X=0
    Scale Y=0
    Scale Z=0
    Translate X=0
    Translate Y=0
    Translate Z=0
    Weight=1
   )]}
 RigVMModel {nodes [
   n1 = &RigUnit_BeginExecution()
   n2 = &RigUnit_GetRelativeBoneTransform(Bone=upperarm_r, Space=spine_05)
   n3 = /Script/ControlRig.RigUnit_ModifyBoneTransforms(
    BoneToModify=((Bone="dyn_parent_collar_side_r",Transform=(Rotation=(X=0,Y=0,Z=0,W=1),Translation=(X=5,Y=0,Z=0),Scale3D=(X=1,Y=1,Z=1))))
    Mode=AdditiveLocal
    WeightMaximum=1
    WeightMinimum=0
    ExecutePin=@n1.ExecutePin
    Weight=&RigVMFunction_MathFloatRemap(
     SourceMaximum=-150
     SourceMinimum=0
     TargetMinimum=0
     bClamp=False
     Value=&RigVMFunction_MathQuaternionToEuler(RotationOrder=XYZ, Value=@n2.Transform.Rotation).Result.Z
    ).Result
   ) {is=struct RigVMUnitNode}
   Branch_RigVMUnitNode = &RigVMFunction_ControlFlowBranch(ExecuteContext=@n3.ExecutePin, Condition=RigVMVariableNode() {variable=FrontEnd}.Value)
   n4 = &RigUnit_SphericalPoseReader(
    ActiveRegionScaleFactors=(PositiveWidth=5,NegativeWidth=5,PositiveHeight=1.5,NegativeHeight=2)
    ActiveRegionSize=0.05
    Debug=(bDrawDebug=True,bDraw2D=false,bDrawLocalAxes=true,DebugScale=25,DebugSegments=20,DebugThickness=0.25)
    DriverAxis=(X=1,Y=0,Z=0)
    DriverItem=(Type=Bone,Name="head")
    FalloffRegionScaleFactors=(PositiveWidth=0.5,NegativeWidth=0.5,PositiveHeight=0.5,NegativeHeight=0.5)
    FalloffSize=0.5
    FlipWidthScaling=true
    OptionalParentItem=(Type=Bone,Name="pelvis")
    RotationOffset=(X=-30,Y=-90,Z=0)
    ExecutePin=@Branch_RigVMUnitNode.False
   )
   > Set Transforms for Bone_1_1_1_1_1_4_2 = RigVMFunctionReferenceNode(
    Joint=(Type=Bone,Name="dyn_driver_collar_bck_l")
    Rotate X=0
    Rotate Y=0
    Rotate Z=-25
    Scale X=0
    Scale Y=0
    Scale Z=0
    Translate X=0
    Translate Y=0
    Translate Z=0
    Weight=@n4.OutputParam
   )
   > Set Transforms for Bone_1_1_1_1_1_1_5_3 = RigVMFunctionReferenceNode(
    Joint=(Type=Bone,Name="dyn_driver_collar_bck_r")
    Rotate X=0
    Rotate Y=0
    Rotate Z=-25
    Scale X=0
    Scale Y=0
    Scale Z=0
    Translate X=0
    Translate Y=0
    Translate Z=0
    Weight=@n4.OutputParam
   )
   > Set Transforms for Bone_1_1_1_1_1_1_1_3 = RigVMFunctionReferenceNode(
    Joint=(Type=Bone,Name="dyn_driver_collar_side_r")
    Rotate X=0
    Rotate Y=0
    Rotate Z=-5
    Scale X=0
    Scale Y=0
    Scale Z=0
    Translate X=0
    Translate Y=0
    Translate Z=0
    Weight=@n4.OutputParam
   )
   > n5 = &RigUnit_SphericalPoseReader(
    ActiveRegionScaleFactors=(PositiveWidth=5,NegativeWidth=5,PositiveHeight=1,NegativeHeight=1)
    ActiveRegionSize=0.05
    Debug=(bDrawDebug=false,bDraw2D=false,bDrawLocalAxes=false,DebugScale=25,DebugSegments=20,DebugThickness=0.25)
    DriverAxis=(X=1,Y=0,Z=0)
    DriverItem=(Type=Bone,Name="head")
    FalloffRegionScaleFactors=(PositiveWidth=0.5,NegativeWidth=0.5,PositiveHeight=0.2,NegativeHeight=0.2)
    FalloffSize=0.5
    FlipWidthScaling=true
    OptionalParentItem=(Type=Bone,Name="pelvis")
    RotationOffset=(X=40,Y=-90,Z=0)
   )
   > Set Transforms for Bone_1_1_1_1_1_4_2_1 = RigVMFunctionReferenceNode(
    Joint=(Type=Bone,Name="dyn_driver_collar_bck_l")
    Rotate X=0
    Rotate Y=0
    Rotate Z=10
    Scale X=0
    Scale Y=0
    Scale Z=0
    Translate X=0
    Translate Y=0
    Translate Z=0
    Weight=@n5.OutputParam
   )
   > Set Transforms for Bone_1_1_1_1_1_1_5_3_1 = RigVMFunctionReferenceNode(
    Joint=(Type=Bone,Name="dyn_driver_collar_bck_r")
    Rotate X=0
    Rotate Y=0
    Rotate Z=5
    Scale X=0
    Scale Y=0
    Scale Z=0
    Translate X=0
    Translate Y=0
    Translate Z=0
    Weight=@n5.OutputParam
   )
   Set Transforms for Bone_1_1_1_1_1_4_2_1_1 = RigVMFunctionReferenceNode(
    Joint=(Type=Bone,Name="dyn_driver_collar_bck_l")
    Rotate X=0
    Rotate Y=0
    Rotate Z=10
    Scale X=0
    Scale Y=0
    Scale Z=0
    Translate X=0
    Translate Y=0
    Translate Z=0
    Weight=1
    ExecuteContext=@Branch_RigVMUnitNode.True
   )
   > Set Transforms for Bone_1_1_1_1_1_1_5_3_1_1 = RigVMFunctionReferenceNode(
    Joint=(Type=Bone,Name="dyn_driver_collar_bck_r")
    Rotate X=0
    Rotate Y=0
    Rotate Z=10
    Scale X=0
    Scale Y=0
    Scale Z=0
    Translate X=0
    Translate Y=0
    Translate Z=0
    Weight=1
   )
   /Script/ControlRig.RigUnit_DebugTransformMutable(
    Color=(R=1,G=0,B=0,A=1)
    DebugDrawSettings=(DepthPriority=SDPG_Foreground,Lifetime=-1)
    ExecutePin=()
    Mode=Axes
    Scale=5
    Space=None
    Thickness=0
    WorldOffset=(Rotation=(X=0,Y=0,Z=0,W=1),Translation=(X=0,Y=0,Z=0),Scale3D=(X=1,Y=1,Z=1))
    bEnabled=True
    Transform=&RigUnit_GetBoneTransform(Bone=dyn_collar_bck_01_r).Transform
   ) {is=struct RigVMUnitNode}
   > &RigUnit_ModifyBoneTransforms(
    BoneToModify="((Bone=\"dyn_driver_collar_bck_l\",Transform=(Rotation=(X=0,Y=0,Z=0,W=1),Translation=(X=0,Y=-3.5,Z=0),Scale3D=(X=1,Y=1,Z=1))),(Bone=\"dyn_driver_collar_bck_r\",Transform=(Rotation=(X=0,Y=0,Z=0,W=1),Translation=(X=0,Y=-5,Z=4),Scale3D=(X=1,Y=1,Z=1))),(Bone=\"dyn_ ...(truncated)"
    Mode=AdditiveLocal
    Weight=1
    WeightMaximum=1
    WeightMinimum=0
    BoneToModify.1.Transform.Rotation=&RigVMFunction_MathQuaternionFromEuler(Euler=(X=0,Y=0,Z=-10)).Result
    BoneToModify.0.Transform.Rotation=&RigVMFunction_MathQuaternionFromEuler(Euler=(X=0,Y=0,Z=0)).Result
    BoneToModify.2.Transform.Rotation=&RigVMFunction_MathQuaternionFromEuler(Euler=(X=0,Y=0,Z=30)).Result
    BoneToModify.3.Transform.Rotation=&RigVMFunction_MathQuaternionFromEuler(Euler=(X=0,Y=0,Z=30)).Result
   )
   > &RigUnit_ModifyBoneTransforms(
    BoneToModify=((Bone="dyn_parent_collar_frt_l",Transform=(Rotation=(X=0,Y=0,Z=0,W=1),Translation=(X=-1,Y=0,Z=0),Scale3D=(X=1,Y=1,Z=1))),(Bone="dyn_parent_collar_side_l",Transform=(Rotation=(X=0,Y=0,Z=0,W=1),Translation=(X=-2,Y=0,Z=0),Scale3D=(X=1,Y=1,Z=1))))
    Mode=AdditiveLocal
    Weight=1
    WeightMaximum=1
    WeightMinimum=0
    BoneToModify.0.Transform.Rotation=&RigVMFunction_MathQuaternionFromEuler(Euler=(X=0,Y=-10,Z=0), RotationOrder=XYZ).Result
    BoneToModify.1.Transform.Rotation=&RigVMFunction_MathQuaternionFromEuler(Euler=(X=0,Y=-10,Z=0), RotationOrder=XYZ).Result
   )
   > &RigUnit_ModifyBoneTransforms(
    BoneToModify=((Bone="dyn_driver_collar_bck_l",Transform=(Rotation=(X=0,Y=0,Z=0,W=1),Translation=(X=0,Y=0,Z=0),Scale3D=(X=1,Y=1,Z=1))),(Bone="dyn_driver_collar_bck_r",Transform=(Rotation=(X=0,Y=0,Z=0,W=1),Translation=(X=0,Y=0,Z=0),Scale3D=(X=1,Y=1,Z=1))))
    Mode=AdditiveLocal
    Weight=1
    WeightMaximum=1
    WeightMinimum=0
    BoneToModify.1.Transform.Rotation=&RigVMFunction_MathQuaternionFromEuler(Euler=(X=0,Y=-30,Z=-20)).Result
    BoneToModify.0.Transform.Rotation=&RigVMFunction_MathQuaternionFromEuler(Euler=(X=0,Y=0,Z=20)).Result
   )
   n6 = &RigVMFunction_MathQuaternionToEuler(RotationOrder=XYZ, Value=&RigUnit_GetRelativeBoneTransform(Bone=head, Space=neck_01).Transform.Rotation)
   n7 = &RigVMFunction_MathFloatRemap(SourceMaximum=-25, SourceMinimum=0, TargetMinimum=0, bClamp=true, Value=@n6.Result.X)
   &RigUnit_ModifyBoneTransforms(
    BoneToModify="((Bone=\"dyn_driver_collar_bck_l\",Transform=(Rotation=(X=0,Y=0,Z=0,W=1),Translation=(X=0,Y=0,Z=0),Scale3D=(X=1,Y=1,Z=1))),(Bone=\"dyn_driver_collar_bck_r\",Transform=(Rotation=(X=0,Y=0,Z=0,W=1),Translation=(X=0,Y=0,Z=0),Scale3D=(X=1,Y=1,Z=1))),(Bone=\"dyn_driv ...(truncated)"
    ExecutePin=()
    Mode=AdditiveLocal
    WeightMaximum=1
    WeightMinimum=0
    Weight=@n7.Result
    BoneToModify.2.Transform.Rotation=&RigVMFunction_MathQuaternionFromEuler(Euler=(X=-10,Y=0,Z=-25)).Result
    BoneToModify.0.Transform.Rotation=&RigVMFunction_MathQuaternionFromEuler(Euler=(X=0,Y=15,Z=-30)).Result
   )
   n8 = &RigVMFunction_MathFloatRemap(SourceMaximum=-49, SourceMinimum=0, TargetMinimum=0, bClamp=False, Value=@n6.Result.Z)
   > &RigUnit_ModifyBoneTransforms(
    BoneToModify="((Bone=\"dyn_driver_collar_bck_l\",Transform=(Rotation=(X=0,Y=0,Z=0,W=1),Translation=(X=0,Y=0,Z=0),Scale3D=(X=1,Y=1,Z=1))),(Bone=\"dyn_driver_collar_bck_r\",Transform=(Rotation=(X=0,Y=0,Z=0,W=1),Translation=(X=0,Y=0,Z=0),Scale3D=(X=1,Y=1,Z=1))),(Bone=\"dyn_driv ...(truncated)"
    Mode=AdditiveLocal
    WeightMaximum=1
    WeightMinimum=0
    Weight=/Script/RigVM.RigVMFunction_MathFloatClamp(
     Maximum=1
     Minimum=0
     Value=/Script/RigVM.RigVMFunction_MathFloatSub(A=@n7.Result, B=@n8.Result) {is=struct RigVMUnitNode, template="Subtract::Execute(in A,in B,out Result)"}.Result
    ) {is=struct RigVMUnitNode, template="Clamp::Execute(in Value,in Minimum,in Maximum,out Result)"}.Result
    BoneToModify.1.Transform.Rotation=&RigVMFunction_MathQuaternionFromEuler(Euler=(X=0,Y=0,Z=0)).Result
   )
   > &RigUnit_DebugTransformMutable(
    Color=(R=1,G=0,B=0,A=1)
    DebugDrawSettings=(DepthPriority=SDPG_Foreground,Lifetime=-1)
    Mode=Axes
    Scale=8
    Space=None
    Thickness=0.5
    WorldOffset=(Rotation=(X=0,Y=0,Z=0,W=1),Translation=(X=0,Y=0,Z=0),Scale3D=(X=1,Y=1,Z=1))
    bEnabled=True
    Transform=&RigUnit_GetBoneTransform(Bone=dyn_collar_bck_01_r).Transform
   )
   > &RigUnit_DebugTransformMutable(
    Color=(R=1,G=0,B=0,A=1)
    DebugDrawSettings=(DepthPriority=SDPG_Foreground,Lifetime=-1)
    Mode=Axes
    Scale=8
    Space=None
    Thickness=0.5
    WorldOffset=(Rotation=(X=0,Y=0,Z=0,W=1),Translation=(X=0,Y=0,Z=0),Scale3D=(X=1,Y=1,Z=1))
    bEnabled=True
    Transform=&RigUnit_GetBoneTransform(Bone=dyn_collar_bck_01_l).Transform
   )
   &RigUnit_DebugTransformMutable(
    Color=(R=1,G=0,B=0,A=1)
    DebugDrawSettings=(DepthPriority=SDPG_Foreground,Lifetime=-1)
    ExecutePin=()
    Mode=Axes
    Scale=10
    Space=None
    Thickness=0
    WorldOffset=(Rotation=(X=0,Y=0,Z=0,W=1),Translation=(X=0,Y=0,Z=0),Scale3D=(X=1,Y=1,Z=1))
    bEnabled=True
    Transform=&RigUnit_GetBoneTransform(Bone=dyn_parent_collar_side_r).Transform
   )
   &RigUnit_DebugTransformMutable(
    Color=(R=1,G=0,B=0,A=1)
    DebugDrawSettings=(DepthPriority=SDPG_Foreground,Lifetime=-1)
    ExecutePin=()
    Mode=Axes
    Scale=10
    Space=None
    Thickness=1
    WorldOffset=(Rotation=(X=0,Y=0,Z=0,W=1),Translation=(X=0,Y=0,Z=0),Scale3D=(X=1,Y=1,Z=1))
    bEnabled=True
    Transform=&RigUnit_GetBoneTransform(Bone=dyn_parent_collar_side_r).Transform
   )
   &RigVMFunction_MathFloatRemap(
    SourceMaximum=-30
    SourceMinimum=0
    TargetMinimum=0
    bClamp=False
    Value=&RigVMFunction_MathQuaternionToEuler(RotationOrder=XYZ, Value=&RigUnit_GetRelativeBoneTransform(Bone=head, Space=spine_05).Transform.Rotation).Result.Z
   )
   &RigVMFunction_MathFloatRemap(
    SourceMaximum=-15
    SourceMinimum=0
    TargetMinimum=0
    bClamp=False
    Value=&RigVMFunction_MathQuaternionToEuler(RotationOrder=XYZ, Value=&RigUnit_GetRelativeBoneTransform(Bone=head, Space=neck_01).Transform.Rotation).Result.Y
   )
   &RigVMFunction_MathFloatRemap(
    SourceMaximum=-30
    SourceMinimum=0
    TargetMinimum=0
    bClamp=False
    Value=&RigVMFunction_MathQuaternionToEuler(RotationOrder=XYZ, Value=&RigUnit_GetRelativeBoneTransform(Bone=head, Space=spine_05).Transform.Rotation).Result.Z
   )
   &RigVMFunction_MathFloatRemap(
    SourceMaximum=-30
    SourceMinimum=0
    TargetMinimum=0
    bClamp=False
    Value=&RigVMFunction_MathQuaternionToEuler(RotationOrder=XYZ, Value=&RigUnit_GetRelativeBoneTransform(Bone=upperarm_l, Space=spine_05).Transform.Rotation).Result.Z
   )
   &RigVMFunction_MathQuaternionFromEuler(Euler=(X=50,Y=0,Z=175))
   &RigVMFunction_MathQuaternionFromEuler(Euler=(X=80,Y=15,Z=10))
   n9 = &RigUnit_GetRelativeBoneTransform(Bone=clavicle_l, Space=spine_05)
   &RigUnit_ModifyBoneTransforms(
    BoneToModify=((Bone="dyn_collar_frt_01",Transform=(Rotation=(X=0,Y=0,Z=0,W=1),Translation=(X=0,Y=5,Z=5),Scale3D=(X=1,Y=1,Z=1))))
    ExecutePin=()
    Mode=AdditiveLocal
    WeightMaximum=1
    WeightMinimum=0
    Weight=&RigVMFunction_MathFloatRemap(
     SourceMaximum=4
     SourceMinimum=75
     TargetMinimum=0
     bClamp=False
     Value=&RigVMFunction_MathQuaternionToEuler(RotationOrder=XZY, Value=@n9.Transform.Rotation).Result.Y
    ).Result
    BoneToModify.0.Transform.Rotation=&RigVMFunction_MathQuaternionFromEuler(Euler=(X=30,Y=0,Z=160)).Result
   )
   n10 = &RigVMFunction_MathQuaternionToEuler(RotationOrder=XYZ, Value=&RigUnit_GetRelativeBoneTransform(Bone=head, Space=neck_01).Transform.Rotation)
   n11 = &RigVMFunction_MathFloatRemap(SourceMaximum=-49, SourceMinimum=0, TargetMinimum=0, bClamp=False, Value=@n10.Result.Z)
   n12 = &RigVMFunction_MathFloatRemap(SourceMaximum=-25, SourceMinimum=0, TargetMinimum=0, bClamp=true, Value=@n10.Result.Y)
   &RigUnit_ModifyBoneTransforms(
    BoneToModify=((Bone="dyn_driver_collar_bck_l",Transform=(Rotation=(X=0,Y=0,Z=0,W=1),Translation=(X=0,Y=0,Z=0),Scale3D=(X=1,Y=1,Z=1))),(Bone="dyn_driver_collar_bck_r",Transform=(Rotation=(X=0,Y=0,Z=0,W=1),Translation=(X=0,Y=0,Z=0),Scale3D=(X=1,Y=1,Z=1))))
    ExecutePin=()
    Mode=AdditiveLocal
    WeightMaximum=1
    WeightMinimum=0
    Weight=/Script/RigVM.RigVMFunction_MathFloatSelectBool(
     Condition=/Script/RigVM.RigVMFunction_MathFloatGreater(B=@n11.Result, A=@n12.Result) {is=struct RigVMUnitNode, template="Greater::Execute(in A,in B,out Result)"}.Result
     IfFalse=@n11.Result
     IfTrue=@n12.Result
    ) {is=struct RigVMUnitNode}.Result
    BoneToModify.0.Transform.Rotation=&RigVMFunction_MathQuaternionFromEuler(Euler=(X=0,Y=0,Z=-55)).Result
    BoneToModify.1.Transform.Rotation=&RigVMFunction_MathQuaternionFromEuler(Euler=(X=0,Y=0,Z=-55)).Result
   )
  ]}
]
hierarchy [
 root {loc=<0 0 0>, rot=<0 0 0 1>}
 pelvis {parent=root, loc=<0 -0.5301 98.0571>, rot=<0.0127 0.707 -0.0127 -0.707>}
 spine_01 {parent=pelvis, loc=<6.1208 -0.9328 0>, rot=<0 0 0.034 -0.9994>}
 spine_02 {parent=spine_01, loc=<6.5996 -0.2095 0>, rot=<0 0 0.0238 -0.9997>}
 spine_03 {parent=spine_02, loc=<6.4914 -0.1237 0>, rot=<0 0 0.0149 -0.9999>}
 spine_04 {parent=spine_03, loc=<6.9598 0.0794 0>, rot=<0 0 -0.0132 -0.9999>}
 spine_05 {parent=spine_04, loc=<6.9987 0 0>, rot=<0 0 -0.0015 -1>}
 clavicle_l {parent=spine_05, loc=<5.4831 -0.1053 -3.2946>, rot=<0.0465 -0.7289 -0.009 -0.6829>}
 upperarm_l {parent=clavicle_l, loc=<9.8658 -1.2136 -0.4105>, rot=<0.0164 -0.3245 -0.0269 -0.9454>}
 lowerarm_l {parent=upperarm_l, loc=<26.465 -0.8345 0.259>, rot=<0.0036 0.0161 -0.2298 -0.9731>}
 hand_l {parent=lowerarm_l, loc=<22.7355 -0.3203 -0.6105>, rot=<-0.6784 -0.0223 0.0735 -0.7307>}
 index_metacarpal_l {parent=hand_l, loc=<1.6535 0.0183 -2.96>, rot=<0.0158 -0.0124 -0.0173 -0.9996>}
 index_01_l {parent=index_metacarpal_l, loc=<8.3008 0.0357 -0.0524>, rot=<-0.0044 0.0485 0.0924 -0.9945>}
 index_02_l {parent=index_01_l, loc=<3.3587 -0.3003 0.0452>, rot=<-0.002 0.0027 0.3225 -0.9466>}
 index_03_l {parent=index_02_l, loc=<2.5578 -0.123 -0.0252>, rot=<-0.002 0.0067 0.2558 -0.9667>}
 middle_metacarpal_l {parent=hand_l, loc=<2.1856 -0.1725 -0.5984>, rot=<0.006 0.0168 -0.0329 -0.9993>}
 middle_01_l {parent=middle_metacarpal_l, loc=<7.3275 -0.2447 0.0129>, rot=<-0.0208 0.0042 0.0858 -0.9961>}
 middle_02_l {parent=middle_01_l, loc=<4.0123 0.1401 0.1175>, rot=<-0.0218 0.0258 0.3512 -0.9357>}
 middle_03_l {parent=middle_02_l, loc=<2.6077 0.0534 0.1373>, rot=<0.0082 0.0188 0.1708 -0.9851>}
 pinky_metacarpal_l {parent=hand_l, loc=<1.7616 -0.9753 3.3814>, rot=<-0.0397 0.0814 -0.069 -0.9935>}
 pinky_01_l {parent=pinky_metacarpal_l, loc=<7.1709 -0.4831 -0.1094>, rot=<-0.0654 -0.0035 0.2242 -0.9723>}
 pinky_02_l {parent=pinky_01_l, loc=<2.8274 0.0893 -0.1043>, rot=<0.0037 0.0011 0.2854 -0.9584>}
 pinky_03_l {parent=pinky_02_l, loc=<2.2046 0.0307 -0.0078>, rot=<-0.0002 0.0098 0.1445 -0.9895>}
 ring_metacarpal_l {parent=hand_l, loc=<2.2074 -0.3625 1.3638>, rot=<-0.0284 0.0722 -0.0458 -0.9959>}
 ring_01_l {parent=ring_metacarpal_l, loc=<7.0295 -0.6197 -0.1446>, rot=<-0.0489 -0.0252 0.1892 -0.9804>}
 ring_02_l {parent=ring_01_l, loc=<3.6174 0.4647 -0.1027>, rot=<0 0 0.2806 -0.9598>}
 ring_03_l {parent=ring_02_l, loc=<2.4152 0.1141 0.0199>, rot=<0 0 0.1189 -0.9929>}
 thumb_01_l {parent=hand_l, loc=<1.3118 -0.3191 -2.459>, rot=<0.6546 -0.4042 0.0664 -0.6354>}
 thumb_02_l {parent=thumb_01_l, loc=<5.9336 -0.0448 0.0401>, rot=<-0.0037 0.0213 0.298 -0.9543>}
 thumb_03_l {parent=thumb_02_l, loc=<2.2223 0.3682 -0.0907>, rot=<0.0118 -0.013 0.0066 -0.9998>}
 lowerarm_twist_01_l {parent=lowerarm_l, loc=<18.316 0 0>, rot=<0 0 0 1>}
 lowerarm_twist_02_l {parent=lowerarm_l, loc=<10.6707 0 0>, rot=<0 0 0 1>}
 dyn_parent_cuff_01_l {parent=lowerarm_l, loc=<17.2785 -0.3076 -0.013>, rot=<0.0722 0.3332 0.2199 -0.914>}
 dyn_cuff_01_l {parent=dyn_parent_cuff_01_l, loc=<0 0 0>, rot=<0.4157 -0.6165 0.6654 0.0656>}
 upperarm_twist_01_l {parent=upperarm_l, loc=<7.2135 -0.1134 0>, rot=<0 0 0 1>}
 upperarm_twist_02_l {parent=upperarm_l, loc=<19.7998 -0.1134 0>, rot=<0 0 0 1>}
 clavicle_r {parent=spine_05, loc=<5.4831 -0.1053 3.2946>, rot=<0.7289 0.0465 -0.6829 0.009>}
 upperarm_r {parent=clavicle_r, loc=<-9.8658 1.2136 0.4104>, rot=<0.0164 -0.3245 -0.0269 -0.9454>}
 lowerarm_r {parent=upperarm_r, loc=<-26.4651 0.8345 -0.259>, rot=<0.0036 0.0161 -0.2298 -0.9731>}
 hand_r {parent=lowerarm_r, loc=<-22.7356 0.3203 0.6106>, rot=<-0.6784 -0.0223 0.0735 -0.7307>}
 index_metacarpal_r {parent=hand_r, loc=<-1.6534 -0.0181 2.9601>, rot=<0.0158 -0.0124 -0.0173 -0.9996>}
 index_01_r {parent=index_metacarpal_r, loc=<-8.3008 -0.0357 0.0524>, rot=<-0.0044 0.0485 0.0924 -0.9945>}
 index_02_r {parent=index_01_r, loc=<-3.3589 0.3003 -0.0453>, rot=<-0.002 0.0027 0.3225 -0.9466>}
 index_03_r {parent=index_02_r, loc=<-2.5578 0.123 0.0252>, rot=<-0.002 0.0067 0.2558 -0.9667>}
 middle_metacarpal_r {parent=hand_r, loc=<-2.1856 0.1726 0.5984>, rot=<0.006 0.0168 -0.0329 -0.9993>}
 middle_01_r {parent=middle_metacarpal_r, loc=<-7.3273 0.2447 -0.0129>, rot=<-0.0208 0.0042 0.0858 -0.9961>}
 middle_02_r {parent=middle_01_r, loc=<-4.0124 -0.14 -0.1175>, rot=<-0.0218 0.0258 0.3512 -0.9357>}
 middle_03_r {parent=middle_02_r, loc=<-2.6077 -0.0533 -0.1373>, rot=<0.0082 0.0188 0.1708 -0.9851>}
 pinky_metacarpal_r {parent=hand_r, loc=<-1.7617 0.9753 -3.3814>, rot=<-0.0397 0.0814 -0.069 -0.9935>}
 pinky_01_r {parent=pinky_metacarpal_r, loc=<-7.1707 0.4832 0.1094>, rot=<-0.0654 -0.0035 0.2242 -0.9723>}
 pinky_02_r {parent=pinky_01_r, loc=<-2.8274 -0.0892 0.1043>, rot=<0.0037 0.0011 0.2854 -0.9584>}
 pinky_03_r {parent=pinky_02_r, loc=<-2.2045 -0.031 0.0078>, rot=<-0.0002 0.0098 0.1445 -0.9895>}
 ring_metacarpal_r {parent=hand_r, loc=<-2.2073 0.3627 -1.3638>, rot=<-0.0284 0.0722 -0.0458 -0.9959>}
 ring_01_r {parent=ring_metacarpal_r, loc=<-7.0297 0.6197 0.1446>, rot=<-0.0489 -0.0252 0.1892 -0.9804>}
 ring_02_r {parent=ring_01_r, loc=<-3.6172 -0.4647 0.1027>, rot=<0 0 0.2806 -0.9598>}
 ring_03_r {parent=ring_02_r, loc=<-2.4152 -0.1141 -0.0199>, rot=<0 0 0.1189 -0.9929>}
 thumb_01_r {parent=hand_r, loc=<-1.3119 0.3192 2.459>, rot=<0.6546 -0.4042 0.0664 -0.6354>}
 thumb_02_r {parent=thumb_01_r, loc=<-5.9336 0.0448 -0.0401>, rot=<-0.0037 0.0213 0.298 -0.9543>}
 thumb_03_r {parent=thumb_02_r, loc=<-2.2223 -0.3682 0.0907>, rot=<0.0118 -0.013 0.0066 -0.9998>}
 lowerarm_twist_01_r {parent=lowerarm_r, loc=<-18.3161 0 0>, rot=<0 0 0 1>}
 lowerarm_twist_02_r {parent=lowerarm_r, loc=<-10.6707 0 0>, rot=<0 0 0 1>}
 dyn_parent_cuff_01_r {parent=lowerarm_r, loc=<-17.2784 0.3076 0.0131>, rot=<0.914 -0.2199 0.3332 0.0722>}
 dyn_cuff_01_r {parent=dyn_parent_cuff_01_r, loc=<0 0 0>, rot=<0.6654 0.0656 0.4157 -0.6165>}
 upperarm_twist_01_r {parent=upperarm_r, loc=<-7.2135 0.1134 0>, rot=<0 0 0 1>}
 upperarm_twist_02_r {parent=upperarm_r, loc=<-19.7998 0.1134 0>, rot=<0 0 0 1>}
 neck_01 {parent=spine_05, loc=<11.3138 0.1529 0>, rot=<0 0 -0.2003 -0.9797>}
 neck_02 {parent=neck_01, loc=<4.5627 0 0>, rot=<0 0 0.0504 -0.9987>}
 head {parent=neck_02, loc=<5.3914 0 0>, rot=<0 0 0.115 -0.9934>}
 dyn_clav_flap_r {parent=spine_05, loc=<12.1164 -3.4135 7.9211>, rot=<0.0284 -0.7065 -0.0284 -0.7065>}
 dyn_parent_collar_frt_l {parent=spine_05, loc=<9.6149 5.2364 -9.7361>, rot=<0.0284 -0.7065 -0.0284 -0.7065>}
 dyn_driver_collar_frt_l {parent=dyn_parent_collar_frt_l, loc=<0 0 0>, rot=<-0.5849 0.3422 -0.4295 -0.597>}
 dyn_collar_frt_01_l {parent=dyn_driver_collar_frt_l, loc=<0 0 0>, rot=<0 0 0 1>}
 dyn_parent_collar_side_l {parent=spine_05, loc=<10.9919 -1.5535 -10.5154>, rot=<0.0284 -0.7065 -0.0284 -0.7065>}
 dyn_driver_collar_side_l {parent=dyn_parent_collar_side_l, loc=<0 0 0>, rot=<-0.5142 0.4612 -0.3669 -0.6231>}
 dyn_collar_side_01_l {parent=dyn_driver_collar_side_l, loc=<0 0 0>, rot=<0 0 0 1>}
 dyn_parent_collar_belt {parent=dyn_collar_side_01_l, loc=<3.6024 -0.8762 -1.9973>, rot=<0.5142 -0.4612 0.3669 -0.6231>}
 dyn_driver_collar_belt {parent=dyn_parent_collar_belt, loc=<0 0 0>, rot=<-0.0481 -0.1176 -0.573 -0.8096>}
 dyn_collar_belt_01 {parent=dyn_driver_collar_belt, loc=<0 0 0>, rot=<0 0 0 1>}
 dyn_parent_collar_bck_l {parent=spine_05, loc=<10.3504 -7.3194 -5.1528>, rot=<0.0284 -0.7065 -0.0284 -0.7065>}
 dyn_driver_collar_bck_l {parent=dyn_parent_collar_bck_l, loc=<0 0 0>, rot=<-0.1496 0.6702 -0.0974 -0.7204>}
 dyn_collar_bck_01_l {parent=dyn_driver_collar_bck_l, loc=<0 0 0>, rot=<0 0 0 1>}
 dyn_parent_collar_bck_r {parent=spine_05, loc=<10.3991 -6.9722 4.6401>, rot=<0.0284 -0.7065 -0.0284 -0.7065>}
 dyn_driver_collar_bck_r {parent=dyn_parent_collar_bck_r, loc=<0 0 0>, rot=<0.0572 0.765 0.132 -0.6277>}
 dyn_collar_bck_01_r {parent=dyn_driver_collar_bck_r, loc=<0 0 0>, rot=<0 0 0 -1>}
 dyn_parent_collar_side_r {parent=spine_05, loc=<10.896 1.1244 11.1731>, rot=<0.0284 -0.7065 -0.0284 -0.7065>}
 dyn_driver_collar_side_r {parent=dyn_parent_collar_side_r, loc=<0 0 0>, rot=<-0.21 -0.724 -0.6075 0.2503>}
 dyn_collar_side_01_r {parent=dyn_driver_collar_side_r, loc=<0 0 0>, rot=<0 0 0 1>}
 dyn_torso_pouch_l {parent=spine_04, loc=<-9.2423 2.6685 -12.6153>, rot=<0.0294 -0.7065 -0.0294 -0.7065>}
 dyn_parent_pouch_flap_l {parent=dyn_torso_pouch_l, loc=<1.5935 1.3536 3.6708>, rot=<0 0 0 1>}
 dyn_pouch_flap_l {parent=dyn_parent_pouch_flap_l, loc=<0 0 0>, rot=<0.7119 -0.3865 -0.3557 -0.4662>}
 dyn_torso_pouch_r {parent=spine_04, loc=<-9.2423 2.6685 12.6153>, rot=<0.0294 -0.7065 -0.0294 -0.7065>}
 dyn_parent_pouch_flap_r {parent=dyn_torso_pouch_r, loc=<-1.5935 1.3536 3.6711>, rot=<1 0 0 0>}
 dyn_pouch_flap_r {parent=dyn_parent_pouch_flap_r, loc=<0 0 0>, rot=<-0.3865 -0.7119 -0.4662 0.3557>}
 dyn_jacket_r {parent=spine_03, loc=<1.3923 -0.4882 12>, rot=<0.0547 -0.9985 0 0>}
 dyn_jacket_back {parent=spine_03, loc=<4.7108 -7.7952 0.0377>, rot=<0.0547 -0.9985 0 0>}
 dyn_jacket_l {parent=spine_03, loc=<1.3923 -0.4882 -12>, rot=<0.0547 -0.9985 0 0>}
 dyn_parent_tab_bot {parent=dyn_jacket_l, loc=<7.2647 6.9687 1.6731>, rot=<0 0.7071 0 -0.7071>}
 dyn_tab_bot_01 {parent=dyn_parent_tab_bot, loc=<0 0 0>, rot=<0 0 0 1>}
 dyn_tab_bot_02 {parent=dyn_tab_bot_01, loc=<-0.5929 2.2581 0.0003>, rot=<0 0 0 1>}
 thigh_l {parent=pelvis, loc=<-1.0855 -0.2109 -8.7053>, rot=<0.0009 0.0355 0.0155 -0.9993>}
 calf_l {parent=thigh_l, loc=<-42.5284 -0.0019 -0.7297>, rot=<0.0006 -0.0161 -0.0391 -0.9991>}
 calf_twist_01_l {parent=calf_l, loc=<-35.8483 -1.9197 -0.5794>, rot=<0.0237 0.0095 -0.0035 -0.9997>}
 calf_twist_02_l {parent=calf_l, loc=<-19.1546 -1.7943 -0.2638>, rot=<0.0237 0.0095 -0.0035 -0.9997>}
 foot_l {parent=calf_l, loc=<-46.7401 -2.5014 -0.592>, rot=<0.0427 -0.0238 0.0353 -0.9982>}
 ball_l {parent=foot_l, loc=<-3.5658 10.8066 -0.1544>, rot=<0.0001 -0.0001 -0.7063 -0.7079>}
 dyn_shoe_flap_l {parent=foot_l, loc=<4.5296 -1.0704 4.2549>, rot=<-0.0357 -0.7045 -0.0284 -0.7082>}
 thigh_twist_01_l {parent=thigh_l, loc=<-27.952 0 0>, rot=<0.0013 0 0 -1>}
 dyn_knee_l {parent=thigh_l, loc=<-42.5284 -0.0019 -0.7297>, rot=<0.0006 -0.0161 -0.0391 -0.9991>}
 dyn_thigh_pouch_top_l {parent=thigh_l, loc=<-5.9008 9.2028 -6.4986>, rot=<-0.0019 -0.7317 0.0015 -0.6816>}
 dyn_thigh_pouch_bot_l {parent=thigh_l, loc=<-15.3539 9.9744 -4.7226>, rot=<-0.0019 -0.7317 0.0015 -0.6816>}
 thigh_r {parent=pelvis, loc=<-1.0853 -0.2109 8.7053>, rot=<0.0355 -0.0009 0.9993 0.0155>}
 calf_r {parent=thigh_r, loc=<42.5286 0.0019 0.7297>, rot=<0.0006 -0.0161 -0.0391 -0.9991>}
 calf_twist_01_r {parent=calf_r, loc=<35.8485 1.9197 0.5794>, rot=<0.0237 0.0095 -0.0035 -0.9997>}
 calf_twist_02_r {parent=calf_r, loc=<19.1546 1.7943 0.2638>, rot=<0.0237 0.0095 -0.0035 -0.9997>}
 foot_r {parent=calf_r, loc=<46.7403 2.5014 0.592>, rot=<0.0427 -0.0238 0.0353 -0.9982>}
 ball_r {parent=foot_r, loc=<3.5658 -10.8066 0.1544>, rot=<0.0001 -0.0001 -0.7063 -0.7079>}
 dyn_shoe_flap_r {parent=foot_r, loc=<-4.5298 1.0703 -4.2547>, rot=<0.7082 0.0284 -0.7045 -0.0357>}
 thigh_twist_01_r {parent=thigh_r, loc=<27.952 0 0>, rot=<0.0013 0 0 -1>}
 dyn_knee_r {parent=thigh_r, loc=<42.528 0.002 0.7297>, rot=<0.0006 -0.0161 -0.0391 -0.9991>}
 dyn_thigh_pouch_r {parent=thigh_r, loc=<34.0828 1.0383 8.9244>, rot=<-0.6816 0.0015 0.7317 0.0019>}
 dyn_thigh_pouch_01_r {parent=dyn_thigh_pouch_r, loc=<-2.0691 -0.0376 4.8682>, rot=<0 0 0 1>}
]
variables [FrontEnd {type=bool, default=False}]
