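# Physics asset for the jacket on the Mascot Militia Cuddle medium female player cosmetic body.
# Three dynamic jacket bones (back, left, right) each carry a small capsule, while spine_03 carries two larger capsules as the torso body.
# Each jacket bone is constrained directly to spine_03 with angular limits: narrow twist on all three, and wider swing on the back than on the sides.
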PhysicsAsset /BRCosmetics/Characters/Player/Female/Medium/Bodies/F_MED_Mascot_Militia_Cuddle/Meshes/F_MED_MascotM_Cuddle_Jacket_Physics
bodies {
 dyn_jacket_back [{shapeType=Capsule, length=3.25, radius=3.75, shapeName=dyn_jacket_back_capsule, center=<4 -1 0>}]
 dyn_jacket_l [{shapeType=Capsule, length=3.25, radius=3.75, shapeName=dyn_jacket_l_capsule, center=<4 0 1.5>, rotation=<rot 0 0 90>}]
 dyn_jacket_r [{shapeType=Capsule, length=3.25, radius=3.75, shapeName=dyn_jacket_r_capsule, center=<4 0 -1.5>, rotation=<rot 0 0 90>}]
 spine_03 [
  {shapeType=Capsule, length=11.75, radius=7.5, center=<-0.1632 3.1765 0.0001>, rotation=<rot 90 -0.5968 -5.5968>}
  {shapeType=Capsule, length=6.4714, radius=6.0286, center=<-2.9695 0.597 -0.0254>, rotation=<rot 0.2247 7.1286 -0.1929>}
 ]
}
constraints [
 {bone1Name=dyn_jacket_back, bone2Name=spine_03, swing1LimitDegrees=25, swing2LimitDegrees=20, twistLimitDegrees=5}
 {bone1Name=dyn_jacket_l, bone2Name=spine_03, swing1LimitDegrees=20, swing2LimitDegrees=10, twistLimitDegrees=5}
 {bone1Name=dyn_jacket_r, bone2Name=spine_03, swing1LimitDegrees=20, swing2LimitDegrees=10, twistLimitDegrees=5}
]
