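# Physics asset for the collar of the female medium Mascot Militia Cuddle cosmetic body, defining collision shapes and joint limits for its dynamic collar bones.
# Four capsule bodies cover the left and right back and side collar bones, each constrained to its own small sphere body on a matching driver bone.
# Swing and twist are limited, more tightly at the back than the sides.
# A large head capsule provides an additional collision body without constraints.
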PhysicsAsset /BRCosmetics/Characters/Player/Female/Medium/Bodies/F_MED_Mascot_Militia_Cuddle/Meshes/F_MED_Mascot_Militia_Cuddle_Collar_Physics
bodies {
 dyn_collar_bck_01_l [{shapeType=Capsule, length=7, radius=1.5, shapeName=dyn_collar_bck_01_l_capsule, center=<6 0 0>, rotation=<rot 90 0 0>}]
 dyn_collar_bck_01_r [{shapeType=Capsule, length=7, radius=1.5, shapeName=dyn_collar_bck_01_r_capsule, center=<6 0 0>, rotation=<rot 90 0 0>}]
 dyn_collar_side_01_l [{shapeType=Capsule, length=10, radius=2, shapeName=dyn_collar_side_01_l_capsule, center=<1.6509 0 0>, rotation=<rot 90 0 0>}]
 dyn_collar_side_01_r [{shapeType=Capsule, length=8.75, radius=2, shapeName=dyn_collar_side_01_r_capsule, center=<1 0 0>, rotation=<rot 90 0 0>}]
 dyn_driver_collar_bck_l [{shapeType=Sphere, radius=0.505}]
 dyn_driver_collar_bck_r [{shapeType=Sphere, radius=0.505}]
 dyn_driver_collar_side_l [{shapeType=Sphere, radius=0.505}]
 dyn_driver_collar_side_r [{shapeType=Sphere, radius=0.505}]
 head [{shapeType=Capsule, length=5.505, radius=10.755, center=<5.9703 -0.3712 -0.0002>, rotation=<rot 0 -90 0>}]
}
constraints [
 {bone1Name=dyn_collar_side_01_l, bone2Name=dyn_driver_collar_side_l, swing1LimitDegrees=25, swing2LimitDegrees=25, twistLimitDegrees=10}
 {bone1Name=dyn_collar_bck_01_l, bone2Name=dyn_driver_collar_bck_l, swing1LimitDegrees=15, swing2LimitDegrees=15, twistLimitDegrees=1}
 {bone1Name=dyn_collar_bck_01_r, bone2Name=dyn_driver_collar_bck_r, swing1LimitDegrees=15, swing2LimitDegrees=15, twistLimitDegrees=1}
 {bone1Name=dyn_collar_side_01_r, bone2Name=dyn_driver_collar_side_r, swing1LimitDegrees=25, swing2LimitDegrees=25, twistLimitDegrees=10}
]
