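# Post-process-style AnimBlueprint for the female medium Militia Cuddle mascot body, adding secondary motion over the pose copied from the parent mesh.
# Collar parent constraints, a front-end-aware Control Rig and alpha-driven knee bone offsets build a cached pose.
# That pose feeds SpeedAlpha-weighted backpack and jacket RigidBody sims, selected by bBackIsStatic, then a collar sim based on spine_05.
# EventGraph is empty and no variables are declared here.
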AnimBlueprint /BRCosmetics/Characters/Player/Female/Medium/Bodies/F_MED_Mascot_Militia_Cuddle/Meshes/F_MED_Mascot_Militia_Cuddle_AnimBP
target_skeleton=&/F_MED_Mascot_Militia_Cuddle_Skeleton
graphs [
 AnimGraph {nodes [
   n1 = Get SpeedAlpha() -> (SpeedAlpha:float)
   n2 = RigidBody(ComponentPose=Local To Component(LocalPose=Use cached pose 'PreRigids'()), Alpha=@n1) {
    bClampLinearTranslationLimitToRefPose=true
    bOverrideWorldGravity=true
    overridePhysicsAsset=&/F_MED_MascotM_Cuddle_BackPack_Physics
    simulationSpace=BaseBoneSpace
    alphaBoolBlend {blendOption=Linear}
    alphaScaleBiasClamp {bInterpResult=true, interpSpeedDecreasing=15, interpSpeedIncreasing=15}
    baseBoneRef {boneName=root}
    componentAppliedLinearAccClamp=<2000 2000 1050>
    componentLinearAccScale=<0.6 0.6 0.3>
    componentLinearVelScale=<0.6 0.6 0.3>
    overrideWorldGravity {z=-900}
   }
   n3 = Component To Local(ComponentPose=@n2)
   n4 = RigidBody(ComponentPose=Local To Component(LocalPose=Use cached pose 'PreRigids'()), Alpha=@n1) {
    bClampLinearTranslationLimitToRefPose=true
    bOverrideWorldGravity=true
    overridePhysicsAsset=&/F_MED_MascotM_Cuddle_Jacket_Physics
    simulationSpace=BaseBoneSpace
    alphaBoolBlend {blendOption=Linear}
    alphaScaleBiasClamp {bInterpResult=true, interpSpeedDecreasing=15, interpSpeedIncreasing=15}
    baseBoneRef {boneName=root}
    componentAppliedLinearAccClamp=<2000 2000 1050>
    componentLinearAccScale=<0.6 0.6 0.3>
    componentLinearVelScale=<0.6 0.6 0.3>
    overrideWorldGravity {z=-900}
   }
   n5 = Component To Local(ComponentPose=@n4)
   n6 = Get bBackIsStatic() -> (bBackIsStatic:bool)
   n7 = RigidBody(ComponentPose=Local To Component(LocalPose=Blend Poses by bool(BlendPose_0=@n3, BlendPose_1=@n5, bActiveValue=@n6) {blendType=Linear})) {
    bClampLinearTranslationLimitToRefPose=true
    bOverrideWorldGravity=true
    overridePhysicsAsset=&/F_MED_Mascot_Militia_Cuddle_Collar_Physics
    simulationSpace=BaseBoneSpace
    alphaBoolBlend {blendOption=Linear}
    alphaScaleBiasClamp {bInterpResult=true, interpSpeedDecreasing=15, interpSpeedIncreasing=15}
    baseBoneRef {boneName=spine_05}
    componentAppliedLinearAccClamp=<2000 2000 1050>
    componentLinearAccScale=<0.6 0.6 0.3>
    componentLinearVelScale=<0.6 0.6 0.3>
    overrideWorldGravity {z=-900}
   }
   Output Pose(Result=Component To Local(ComponentPose=@n7)) {layerGroup=None}
   n8 = "Constraint - dyn_parent_collar_side_l (PP)"!(
    ComponentPose=Local To Component(LocalPose=Copy Pose From Mesh() {bUseAttachedParent=true})
    Alpha=1
    ConstraintWeights_0=0.5
    ConstraintWeights_1:float=0.5
   ) -> (Pose) {
    alphaBoolBlend {blendOption=Linear}
    boneToModify {boneName=dyn_parent_collar_side_l}
    constraintSetup [{targetBone {boneName=clavicle_l}}, {targetBone {boneName=spine_05}}]
    constraintWeights [1,1]
   }
   n9 = "Constraint - dyn_parent_collar_frt_l (PP)"!(ComponentPose=@n8, Alpha=1, ConstraintWeights_0=0.5, ConstraintWeights_1:float=0.5) -> (Pose) {
    alphaBoolBlend {blendOption=Linear}
    boneToModify {boneName=dyn_parent_collar_frt_l}
    constraintSetup [{targetBone {boneName=clavicle_l}}, {targetBone {boneName=spine_05}}]
    constraintWeights [1,1]
   }
   n10 = Control Rig!(Source=Component To Local(ComponentPose=@n9), Alpha, FrontEnd:bool=Get bIsInFrontEnd()) -> (Pose) {
    alphaBoolBlend {blendOption=Linear}
    controlRigAssetReference {blueprintRigClass=&/F_MED_MascotM_Cuddle_CtrlRig.F_MED_MascotM_Cuddle_CtrlRig_C}
   }
   n11 = "Transform (Modify) Bone - Bone: dyn_knee_r"(
    ComponentPose=Local To Component(LocalPose=@n10)
    Alpha=Get CalfRightRotAlpha() -> (CalfRightRotAlpha:float)
    Translation="0.000,-5,3"
    Rotation="0,0,0"
    Scale="1.35,1.35,1.35"
   ) {
    rotationMode=BMM_Additive
    rotationSpace=BCS_BoneSpace
    scaleMode=BMM_Additive
    scaleSpace=BCS_BoneSpace
    translationMode=BMM_Additive
    translationSpace=BCS_BoneSpace
    alphaBoolBlend {blendOption=Linear}
    boneToModify {boneName=dyn_knee_r}
    scale=<1.35 1.35 1.35>
    translation {y=-5, z=3}
   }
   n12 = "Transform (Modify) Bone - Bone: dyn_knee_l"(
    ComponentPose=@n11
    Alpha=Get CalfLeftRotAlpha() -> (CalfLeftRotAlpha:float)
    Translation="0.000,5,-3"
    Rotation="0,0,0"
    Scale="1.35,1.35,1.35"
   ) {
    rotationMode=BMM_Additive
    rotationSpace=BCS_BoneSpace
    scaleMode=BMM_Additive
    scaleSpace=BCS_BoneSpace
    translationMode=BMM_Additive
    translationSpace=BCS_BoneSpace
    alphaBoolBlend {blendOption=Linear}
    boneToModify {boneName=dyn_knee_l}
    scale=<1.35 1.35 1.35>
    translation {y=5, z=-3}
   }
   Save cached pose 'PreRigids'(Pose=Component To Local(ComponentPose=@n12)) {cacheName=PreRigids}
  ]}
 EventGraph {links [], nodes []}
]
variables []
