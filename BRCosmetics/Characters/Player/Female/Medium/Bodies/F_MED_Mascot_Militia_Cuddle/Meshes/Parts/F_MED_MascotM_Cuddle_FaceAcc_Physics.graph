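# Physics asset for the face accessory of the female Cuddle Militia mascot cosmetic, simulating its two ears and a dangling ear tag.
# A head capsule anchors small parent spheres for each ear, each driving a capsule ear with tighter angular limits.
# The two-segment tag chain branches from the left ear's parent through its own sphere.
# Joints into the parent spheres leave swing and twist free, while ear and tag segments carry the limits.
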PhysicsAsset /BRCosmetics/Characters/Player/Female/Medium/Bodies/F_MED_Mascot_Militia_Cuddle/Meshes/Parts/F_MED_MascotM_Cuddle_FaceAcc_Physics
bodies {
 dyn_ear_01_l [{shapeType=Capsule, length=4.25, radius=3.755, center=<3.7701 0 0.0002>, rotation=<rot 90 0 0>}]
 dyn_ear_01_r [{shapeType=Capsule, length=4.5, radius=3.505, center=<3.8577 0 0>, rotation=<rot 90 0 0>}]
 dyn_ear_tag_01 [{shapeType=Capsule, length=3.5, radius=3, center=<0 -0.0001 -2.2876>}]
 dyn_ear_tag_02 [{shapeType=Capsule, length=4.1374, radius=1.0988, center=<0.1314 -0.0152 -2.6284>, rotation=<rot 0.7851 -89.9997 1.031>}]
 dyn_parent_ear_l [{shapeType=Sphere, radius=0.505}]
 dyn_parent_ear_r [{shapeType=Sphere, radius=0.505}]
 dyn_parent_ear_tag [{shapeType=Sphere, radius=0.505}]
 head [{shapeType=Capsule, length=1.25, radius=15, center=<4.5181 0.7139 1.0476>}]
}
constraints [
 {
  bone1Name=dyn_parent_ear_l
  bone2Name=head
  swing1LimitDegrees=45
  swing1Motion=Free
  swing2LimitDegrees=45
  swing2Motion=Free
  twistLimitDegrees=45
  twistMotion=Free
 }
 {
  bone1Name=dyn_parent_ear_r
  bone2Name=head
  swing1LimitDegrees=45
  swing1Motion=Free
  swing2LimitDegrees=45
  swing2Motion=Free
  twistLimitDegrees=45
  twistMotion=Free
 }
 {bone1Name=dyn_ear_01_r, bone2Name=dyn_parent_ear_r, swing1LimitDegrees=45, swing2LimitDegrees=20, twistLimitDegrees=10}
 {bone1Name=dyn_ear_01_l, bone2Name=dyn_parent_ear_l, swing1LimitDegrees=45, swing2LimitDegrees=20, twistLimitDegrees=10}
 {
  bone1Name=dyn_parent_ear_tag
  bone2Name=dyn_parent_ear_l
  swing1LimitDegrees=45
  swing1Motion=Free
  swing2LimitDegrees=45
  swing2Motion=Free
  twistLimitDegrees=45
  twistMotion=Free
 }
 {bone1Name=dyn_ear_tag_01, bone2Name=dyn_parent_ear_tag, swing1LimitDegrees=20, swing2LimitDegrees=20, twistLimitDegrees=15}
 {bone1Name=dyn_ear_tag_02, bone2Name=dyn_ear_tag_01, swing1LimitDegrees=45, swing2LimitDegrees=45, twistLimitDegrees=15}
]
